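# Niagara system for the left-hand sheen effect in the male Desert Shadow emote, pairing a mesh sweep with a sparkle.
# Sheen_Mesh bursts an oriented mesh lasting System.SheenLifetime, animating a dynamic material parameter through a 0-to-1 curve.
# Sparkle bursts a spinning sprite that pops and shrinks, timed by subtracting a system offset from the sheen lifetime.
# Sparkle's skeletal-mesh location module is disabled, leaving it placed by a local spawn offset.
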NiagaraSystem /BRCosmetics/Animation/Game/MainPlayer/Emotes/DesertShadow/NS_Desert_Shadow_Male_Sheen_Emote_LeftHand
dependencies {
 usedDataInterfaces [&NiagaraDataInterfaceCurve]
 usedDynamicInputs []
 usedModules [
  &SystemState
  &self:SystemSpawnScript.SystemScriptSource.SystemScriptGraph.NiagaraNodeAssignment_0.SetVariables_#1
 ]
}
emitters {
 Sheen_Mesh {
  emitterUpdateScript [
   &EmitterState()
   &SpawnBurst_Instantaneous(Use Spawn Probability=false, UseLoopCountLimit=false)
  ]
  particleSpawnScript [
   &InitializeParticle(Lifetime=$System.SheenLifetime, Position Offset Coordinate Space=::Local, UsePositionOffset=true, Write Material Random=true)
   &DynamicMaterialParameters(
    Index 0 Param 2=0.2
    Index 0 Param 3=-0.02
    Module.Param0WriteEnabled=false
    Module.Param1WriteEnabled=true
    Module.Param2WriteEnabled=true
    Module.Param3WriteEnabled=false
   )
   &InitialMeshOrientation(Mesh Axis To Orient=<0 1 0>, Mesh Orientation Mode=::NewEnumerator2, Orientation Coordinate Space=::Local, Use Rotation Vector=true)
  ]
  particleUpdateScript [
   &ParticleState()
   &DynamicMaterialParameters(
    Index 0 Param 1=&FloatFromCurve(FloatCurve=DataInterface() {propertyValues {Curve {keys [{interpMode=RCIM_Cubic, tangentMode=RCTM_User, time=0, value=0}, {interpMode=RCIM_Cubic, tangentMode=RCTM_User, time=1, value=1}]}}})
    Index 0 Param 2=0.4
    Index 0 Param 3=-0.05
    Module.Param0WriteEnabled=true
    Module.Param1WriteEnabled=true
    Module.Param2WriteEnabled=true
    Module.Param3WriteEnabled=false
   ) {module=DynamicMaterialParameters001}
  ]
  rendererClasses [&NiagaraMeshRendererProperties]
  renderers [{rendererClass=&NiagaraMeshRendererProperties, rendererIndex=0}]
 }
 Sparkle {
  emitterUpdateScript [
   &EmitterState()
   &SpawnBurst_Instantaneous(
    Spawn Time=&Subtract_Float(A=$System.SheenLifetime, B=$System.SparkleSpawnTimeOffset)
    Use Spawn Probability=false
    UseLoopCountLimit=false
   )
  ]
  particleSpawnScript [&InitializeParticle(
    Color=<rgba 2 2 2 1>
    Lifetime=&RandomRangeFloat(Minimum=0.666, Maximum=0.777)
    Position Offset=<3 30 3>
    Position Offset Coordinate Space=::Local
    Sprite Rotation Mode=::NewEnumerator2
    Sprite Size=&Multiply_Vector2D_ByFloat(Float=&RandomRangeFloat(Minimum=16, Maximum=23), Vector2D=<1.1 1>)
    Sprite Size Mode=::NewEnumerator4
    UsePositionOffset=true
    Write Material Random=true
   )]
  particleUpdateScript [
   &ParticleState()
   &ScaleSpriteSize(
    Uniform Curve Scale=2
    Uniform Curve Sprite Scale=DataInterface() {propertyValues {Curve {keys [
        {arriveTangent=9.0725, interpMode=RCIM_Cubic, leaveTangent=9.0725, tangentMode=RCTM_User, time=0, value=0}
        {arriveTangent=-1.7917, interpMode=RCIM_Cubic, leaveTangent=-1.7917, tangentMode=RCTM_User, time=0.0666, value=0.9847}
        {arriveTangent=-0.0405, interpMode=RCIM_Cubic, leaveTangent=-0.0405, tangentMode=RCTM_User, time=1, value=0}
       ]}}}
   )
   &SpriteRotationRate(Rotation Rate=&RandomRangeFloat(Minimum=150, Maximum=180))
   &CameraOffset(Camera Offset Amount=2)
   &SkeletalMeshLocation(
    MaskSpawnGroup=false
    Mesh Sampling Type=::NewEnumerator5
    Sampled Position Offset=<0.2 0 0>
    Sampled Position Offset Coordinate Space:ENiagaraSkelMeshTransforms=::NewEnumerator2
    Skeletal Mesh=$User.SkeletalMesh
    Socket Sampling Mode=::NewEnumerator5
   ) {enabled=false}
  ]
  rendererClasses [&NiagaraSpriteRendererProperties]
  renderers [{rendererClass=&NiagaraSpriteRendererProperties}]
 }
}
summary {
 systemName=NS_Desert_Shadow_Male_Sheen_Emote_LeftHand
 emitters [
  {emitterName=Sheen_Mesh, rendererClasses [&NiagaraMeshRendererProperties]}
  {emitterName=Sparkle, rendererClasses [&NiagaraSpriteRendererProperties]}
 ]
 userVariables [User.SkeletalMesh {
   type=NiagaraDataInterfaceSkeletalMesh
   defaultValue {
    struct=NiagaraExt_VariableValue_DataInterface
    value {dataInterface=&self:NiagaraDataInterfaceSkeletalMesh_0}
   }
  }]
}
